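; Burst gameplay cue for a blade hazard hit on a player, tailoring the hit VFX and audio.
; It sets the first Niagara component's HitRotationVelocity: perpendicular to the horizontal cue-location-to-target offset if the target casts to the upright blade hazard, otherwise target forward plus up.
; Each audio component gets an "islocal" flag, true only for a locally viewed Fort player pawn target, before the parent OnBurst runs.
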(fn OnBurst(EitherVFX-preferthisversion)
  (params (MyTarget :type Actor%20Object%20Reference) (Parameters :type Gameplay%20Cue%20Parameters%20Structure%20%28by%20ref%29) (ParticleComponents :type Array%20of%20FXSystem%20Component%20Object%20References) (AudioComponents :type Array%20of%20Audio%20Component%20Object%20References) (BurstCameraShakeInstance :type Camera%20Shake%20Base%20Object%20Reference) (BurstDecalInstance :type Decal%20Actor%20Object%20Reference))
  (bind (_normalizedmagnitude _rawmagnitude _effectcontext _matchedtagname _originaltag _aggregatedsourcetags _aggregatedtargettags _location) (Ability|GameplayCue|BreakGameplayCueParameters Parameters))
  (Utilities|FlowControl|Sequence
    (:then_0
      ;[1 0.3854 0.3854] VFX
      (bind _asniagara_particle_system_component (Utilities|Casting|CastToNiagaraComponent (Utilities|Array|Get(acopy) ParticleComponents)))
      (Utilities|Casting|CastToHelios_BladeHazard_Upright MyTarget
        (:then
          ;[1 1 1] UpRight
          (Effects|Components|ParticleSystem|SetVectorParameter _asniagara_particle_system_component "HitRotationVelocity" (Math|Vector|MakeVector (* (.y (- (Transformation|GetActorLocation MyTarget) _location)) -1.0) (.x (- (Transformation|GetActorLocation MyTarget) _location)))))
        (:CastFailed
          ;[1 1 1] Side
          (Effects|Components|ParticleSystem|SetVectorParameter _asniagara_particle_system_component "HitRotationVelocity" (+ (Transformation|GetActorForwardVector MyTarget) "0.000000,0.000000,1.000000")))))
    (:then_1
      (for _array_element AudioComponents
        (bind _asfort_player_pawn (Utilities|Casting|CastToFortPlayerPawn MyTarget)
          (:then
            (Audio|Parameter|SetBooleanParameter _array_element "islocal" (Pawn|IsLocallyViewed _asfort_player_pawn) :!class /Script/AudioExtensions.AudioParameterControllerInterface))
          (:CastFailed
            (Audio|Parameter|SetBooleanParameter _array_element "islocal" false :!class /Script/AudioExtensions.AudioParameterControllerInterface))))
      (|CallParentFunction MyTarget Parameters ParticleComponents AudioComponents BurstCameraShakeInstance BurstDecalInstance)
      (return))))
