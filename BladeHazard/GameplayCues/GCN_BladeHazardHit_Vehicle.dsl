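; Burst gameplay cue for blade hazard hits involving vehicles, playing impact VFX and a cut sound.
; It sets the Niagara HitRotationVelocity from the target-to-hit offset rotated horizontally when the target casts to an upright blade hazard, otherwise from target forward plus up.
; It spawns the sound at the target, setting "islocal" when the target is a vehicle with a locally viewed passenger, then calls the parent burst.
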(fn CheckForLocalPlayer
  (params (MyTarget :type Actor%20Object%20Reference))
  (bind _asfort_athena_vehicle (Utilities|Casting|CastToFortAthenaVehicle MyTarget))
  (for _array_element (Vehicle|GetAllPassengers _asfort_athena_vehicle)
    (if (Pawn|IsLocallyViewed _array_element)
      (return true))))

(fn OnBurst(EitherVFX-preferthisversion)
  (params (MyTarget :type Actor%20Object%20Reference) (Parameters :type Gameplay%20Cue%20Parameters%20Structure%20%28by%20ref%29) (ParticleComponents :type Array%20of%20FXSystem%20Component%20Object%20References) (AudioComponents :type Array%20of%20Audio%20Component%20Object%20References) (BurstCameraShakeInstance :type Camera%20Shake%20Base%20Object%20Reference) (BurstDecalInstance :type Decal%20Actor%20Object%20Reference))
  (bind (_normalizedmagnitude _rawmagnitude _effectcontext _matchedtagname _originaltag _aggregatedsourcetags _aggregatedtargettags _location) (Ability|GameplayCue|BreakGameplayCueParameters Parameters))
  (Utilities|FlowControl|Sequence
    (:then_0
      ;[1 0.3854 0.3854] VFX
      (bind _asniagara_particle_system_component (Utilities|Casting|CastToNiagaraComponent (Utilities|Array|Get(acopy) ParticleComponents)))
      (Utilities|Casting|CastToHelios_BladeHazard_Upright MyTarget
        (:then
          ;[1 1 1] UpRight
          (Effects|Components|ParticleSystem|SetVectorParameter _asniagara_particle_system_component "HitRotationVelocity" (Math|Vector|MakeVector (* (.y (- (Transformation|GetActorLocation MyTarget) _location)) -1.0) (.x (- (Transformation|GetActorLocation MyTarget) _location)))))
        (:CastFailed
          ;[1 1 1] Side
          (Effects|Components|ParticleSystem|SetVectorParameter _asniagara_particle_system_component "HitRotationVelocity" (+ (Transformation|GetActorForwardVector MyTarget) "0.000000,0.000000,1.000000")))))
    (:then_1
      (bind _islocal (CallFunction|CheckforLocalPlayer (Variables|GetMyTarget)))
      (bind _returnvalue (Audio|SpawnSoundatLocation (Variables|GetMyTarget) "/BladeHazard/Assets/Sound/BladeHazard_Impact_Cut_Vehicle_Cue.BladeHazard_Impact_Cut_Vehicle_Cue" (Transformation|GetWorldLocation (Variables|/Script/Engine.Actor|GetRootComponent (Variables|GetMyTarget)))))
      (Audio|Parameter|SetBooleanParameter _returnvalue "islocal" _islocal :!class /Script/AudioExtensions.AudioParameterControllerInterface)
      (|CallParentFunction MyTarget Parameters ParticleComponents AudioComponents BurstCameraShakeInstance BurstDecalInstance)
      (return))))
